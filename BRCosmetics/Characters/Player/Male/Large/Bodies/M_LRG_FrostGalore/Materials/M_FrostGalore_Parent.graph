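# Parent cosmetic material for the FrostGalore large male body, using the shared 2023 anime shader (Pre, Calculations, Post) with the Cloth shading model.
# An eye animation function feeds lit and shaded color, masked by texture, vertex color or UV1.
# Added emissive combines faint repeating stripes with a blue waveform line, both limited to the FX texture's red channel and half the V range.
# The waveform reuses an audio visualization function, fed by scalar amplitude and spectrum parameters.
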Material /BRCosmetics/Characters/Player/Male/Large/Bodies/M_LRG_FrostGalore/Materials/M_FrostGalore_Parent
graph {
 properties {
  bAutomaticallySetUsageInEditor=false
  bUseMaterialAttributes=true
  bUsedWithMeshDeformer=true
  bUsedWithMorphTargets=true
  bUsedWithSkeletalMesh=true
  bUsedWithStaticLighting=true
  pixelDepthOffsetMode=PDOM_Legacy
  shadingModel=MSM_Cloth
  usedShadingModels=Cloth
 }
 nodes [
  n1 = MaterialFunctionCall!(CustomUV) {materialFunction=/Game/Global/Materials/Character_Parent_2023/Anime_Parent_2023/MaterialFunctions/MFC_AnimeShader_2023_Pre}
  n2 = MaterialFunctionCall!(CustomLightVector, OverrideNormal_WS, STT=@n1.STT, Normal_Map=@n1.STT, DFL=@n1.STT, FX=@n1.STT) {materialFunction=/Game/Global/Materials/Character_Parent_2023/Anime_Parent_2023/MaterialFunctions/MFC_AnimeShader_2023_Calculations}
  n3 = VertexColor()
  n4 = TextureSampleParameter2D() {
   group=01_EyeAnimation_Parameters
   parameterName=EyeMask
   samplerSource=SSM_Wrap_WorldGroupSettings
   samplerType=SAMPLERTYPE_Masks
   sortPriority=-32
   texture=&T_WhiteMask
  }
  n5 = StaticSwitchParameter(True=AppendVector(A=@n3, B=@n3), False=@n4.RGBA) {group=01_EyeAnimation_Parameters, parameterName=UseVertexColorForEyeMask, sortPriority=-33}
  n6 = TextureCoordinate() {coordinateIndex=1}
  n7 = Sign(None=Subtract(A=ComponentMask(None=@n6) {r=true}) {constB=0.0001})
  n8 = MaterialFunctionCall!(
   Color Lit=@n1.Color_Lit
   Color Shaded=@n1.Color_Lit
   Use UV2=StaticBoolParameter() {defaultValue=true, group=01_EyeAnimation_Parameters, parameterName=Use UV2 for Eye Animation, sortPriority=-30}
   Mask=StaticSwitchParameter(
    True=ChannelMaskParameter(None=@n5) {group=01_EyeAnimation_Parameters, parameterName=EyeMaskChannel, sortPriority=-32}.R
    False=Saturate(None=@n7)
   ) {group=01_EyeAnimation_Parameters, parameterName=UseEyeMask, sortPriority=-33}
  ) {materialFunction=/Game/Global/Materials/Character_Parent_2023/Anime_Parent_2023/MaterialFunctions/MF_AnimeEyeAnimation_v2}
  n9 = TextureCoordinate()
  n10 = Frac(None=Multiply(A=ComponentMask(None=@n9) {g=true}) {constB=82})
  n11 = Constant() {r=1}
  n12 = TextureSample() {
   samplerSource=SSM_Wrap_WorldGroupSettings
   samplerType=SAMPLERTYPE_Masks
   texture=/BRCosmetics/Characters/Player/Male/Large/Bodies/M_LRG_FrostGalore/Textures/T_FrostGalore_Body_FX
  }
  n13 = Add(A=ComponentMask(None=TextureCoordinate()) {g=true}) {constB=-0.5}
  n14 = Multiply(A=@n12.R, B=Saturate(None=Multiply(A=@n13) {constB=100}))
  n15 = Multiply(A=Multiply(A=SphereMask(A=@n10, B=@n11) {attenuationRadius=0.12}, B=@n14)) {constB=0.05}
  n16 = Add(A=TextureCoordinate(), B=Constant2Vector() {g=0.1, r=0.025})
  n17 = Subtract(A=Multiply(A=MaterialFunctionCall(XY=@n16) {materialFunction=&Swizzle}.YX) {constB=2}) {constB=0.98}
  n18 = ComponentMask(None=Multiply(A=Abs(None=@n17)) {constB=2}) {g=true}
  n19 = NamedRerouteUsage()
  n20 = TextureCoordinate()
  n21 = MaterialFunctionCall!(
   Audio Amplitude=@n19
   Audio Spectrum=Power(Base=NamedRerouteUsage(), Exp=ScalarParameter() {defaultValue=2, parameterName=SpectrumExp}) {constExponent=4}
   Time
   Freq Base=ScalarParameter() {defaultValue=0.5, group=TechArt, parameterName=StringFreqBase, sortPriority=40}
   Coordinate Gradient=Multiply(A=ComponentMask(None=@n20) {g=true}, B=ScalarParameter() {defaultValue=22, group=TechArt, parameterName=StringGradientLength}) {constB=6}
   Freq Curve Factor=ScalarParameter() {defaultValue=2.1333, group=TechArt, parameterName=StringFreqCurveFactor, sortPriority=40}
   Amplitude Curve Factor=ScalarParameter() {defaultValue=1, group=TechArt, parameterName=StringAmpCurveFactor, sortPriority=40}
   Amplitude Base=ScalarParameter() {defaultValue=0.32, group=TechArt, parameterName=StringAmpBase, sortPriority=40}
   Scroll Base=ScalarParameter() {defaultValue=1, group=TechArt, parameterName=StringScrollBase, sortPriority=40}
   Scroll Curve Factor=ScalarParameter() {defaultValue=0.625, group=TechArt, parameterName=StringScrollCurveFactor, sortPriority=40}
  ) {materialFunction=/BRCosmetics/Characters/Player/Female/Medium/Bodies/F_MED_Tactical_Red_Punk/Skins/Disco/Materials/MF_AudioVisualizationString}
  n22 = Multiply(A=ComponentMask(None=TextureCoordinate()) {g=true}) {constB=0.1}
  n23 = Multiply(A=@n21.Result, B=Saturate(None=OneMinus(None=@n22)))
  n24 = ScalarParameter() {defaultValue=1.8, group=TechArt, parameterName=StringExciteAmt}
  n25 = SphereMask(
   A=Add(A=Frac(None=@n18), B=Multiply(A=@n23, B=@n24))
   B=ScalarParameter() {defaultValue=0.4, group=TechArt, parameterName=StringOffset, sortPriority=38}
   Radius=ScalarParameter() {defaultValue=0.1, group=TechArt, parameterName=StringThickness, sortPriority=38}
  ) {attenuationRadius=1, hardnessPercent=0}
  n26 = ScalarParameter() {group=TechArt, parameterName=StringSeparation, sliderMax=0.5, sortPriority=38}
  n27 = Add(A=@n26)
  n28 = Abs(None=OneMinus(None=Multiply(A=@n25, B=@n27)))
  n29 = Divide(A=Abs(None=Subtract(A=@n28, B=@n26) {constB=0.05}), B=OneMinus(None=@n26))
  n30 = Multiply(A=Power(Base=OneMinus(None=@n29), Exp=ScalarParameter() {defaultValue=120, group=TechArt, parameterName=StringThicknessExp, sortPriority=38}) {constExponent=10}) {constB=100}
  n31 = Multiply(A=Multiply(A=Saturate(None=@n30), B=@n14), B=Constant3Vector() {constant=<rgba 0.0651 0.3756 1.5 1>})
  MaterialFunctionCall!(
   Shading_Gradient=@n2.Shading_Gradient
   Threshold_Gradient=@n2.Shading_Gradient
   Specular_Gradient=@n2.Shading_Gradient
   Sheen_Mask=@n2.Shading_Gradient
   Thickness=@n2.Shading_Gradient
   DFL_Mask=@n2.Shading_Gradient
   FX=@n2.Shading_Gradient
   Metallic_Gradient=@n2.Shading_Gradient
   RimLight_Gradient=@n2.Shading_Gradient
   NormalBuffer=@n2.Shading_Gradient
   --------
   CINE_LightSources=@n2.Shading_Gradient
   CINE_LightColor=@n2.Shading_Gradient
   ---------
   Color_Lit=@n8.Lit
   Color_Shaded=@n8.Lit
   SRM=@n1.SRM
   Emissive=Add(A=Add(A=@n15, B=@n31), B=@n1.Emissive)
   DFL_Color=@n1.SRM
   Specular_Color=@n1.SRM
   RimLight_Color=@n1.SRM
  ) {materialFunction=/Game/Global/Materials/Character_Parent_2023/Anime_Parent_2023/MaterialFunctions/MFC_AnimeShader_2023_Post}
  Multiply(A=@n19) {constB=0.75}
  NamedRerouteDeclaration(None=ScalarParameter() {defaultValue=0.25, parameterName=FrostGalore_LineAmp}) {name=Amplitude, nodeColor=<rgba 0 1 0.1094 1>}
  NamedRerouteDeclaration(None=ScalarParameter() {defaultValue=0.5, parameterName=FrostGalore_LineSpec}) {name=Spectrum, nodeColor=<rgba 1 0 0.75 1>}
 ]
}
