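# Control Rig for the M_LRG_FrostGalore player body, procedurally correcting arm twist and driving costume bones on the legs and laptop.
# The Legs function constrains thigh sphere and robot leg bones, and offsets calf spheres by calf yaw relative to initial pose.
# Upper-arm twist constraints fade out during FEIdle via interpolated weight.
# The laptop base follows spine_03, while its dynamic bones shift when targeting or firearm is set, smoothed by Kalman filter and cubic easing.
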ControlRigBlueprint /BRCosmetics/Characters/Player/Male/Large/Bodies/M_LRG_FrostGalore/Meshes/M_LRG_FrostGalore_CtrlRig
graphs [
 ContainedGraph {nodes [
   n1 = RigVMFunctionEntryNode()
   n2 = &RigUnit_GetTransform(Space=LocalSpace, Item=@n1.Item)
   n3 = &RigUnit_GetTransform(Space=LocalSpace, bInitial=true, Item=@n1.Item)
   > RigVMFunctionReturnNode(OutTransform=/Script/RigVM.RigVMFunction_MathTransformToEulerTransform(Value=&RigVMFunction_MathTransformMakeRelative(Global=@n2.Transform, Parent=@n3.Transform).Local) {is=struct RigVMUnitNode, template="Cast::Execute(in Value,out Result)"}.Result)
  ]}
 New Function_ContainedGraph {nodes [
   RigVMFunctionEntryNode()
   > RigVMFunction_Sequence = RigVMAggregateNode()
   RigVMFunctionReturnNode(ExecuteContext=@RigVMFunction_Sequence.A)
   &RigUnit_ParentConstraint(
    AdvancedSettings=(InterpolationType=Average,RotationOrderForFilter=XZY)
    Child=(Type=Bone,Name="dyn_sphere_thigh_l")
    Filter=(TranslationFilter=(bX=false,bY=true,bZ=false),RotationFilter=(bX=true,bY=true,bZ=true),ScaleFilter=(bX=false,bY=false,bZ=false))
    Parents=((Item=(Type=Bone,Name="thigh_l"),Weight=1))
    Weight=1
    bMaintainOffset=true
    ExecutePin=@RigVMFunction_Sequence.B
   )
   > &RigUnit_ParentConstraint(
    AdvancedSettings=(InterpolationType=Average,RotationOrderForFilter=XZY)
    Child=(Type=Bone,Name="dyn_sphere_thigh_r")
    Filter=(TranslationFilter=(bX=false,bY=true,bZ=false),RotationFilter=(bX=true,bY=true,bZ=true),ScaleFilter=(bX=false,bY=false,bZ=false))
    Parents=((Item=(Type=Bone,Name="thigh_r"),Weight=1))
    Weight=1
    bMaintainOffset=true
   )
   DeformRig_GetDeltaRefTransform_FNC = RigVMFunctionReferenceNode(Item=(Type=Bone,Name="calf_l"), ExecuteContext=@RigVMFunction_Sequence.C)
   > &RigUnit_ModifyTransforms(
    ItemToModify=((Item=(Type=Bone,Name="base_sphere_calf_l"),Transform=(Rotation=(X=0,Y=0,Z=0.284,W=0.9588),Translation=(X=-1,Y=-1.5,Z=0),Scale3D=(X=1,Y=1,Z=1))))
    WeightMinimum=0
    Weight=&RigVMFunction_MathFloatRemap(SourceMaximum=100, SourceMinimum=0, bClamp=true, Value=@DeformRig_GetDeltaRefTransform_FNC.OutTransform.Rotation.Yaw).Result
   )
   > DeformRig_GetDeltaRefTransform_FNC_1 = RigVMFunctionReferenceNode(Item=(Type=Bone,Name="calf_r"))
   > &RigUnit_ModifyTransforms(
    ItemToModify=((Item=(Type=Bone,Name="base_sphere_calf_r"),Transform=(Rotation=(X=0,Y=0,Z=0.284,W=0.9588),Translation=(X=1,Y=1.5,Z=0),Scale3D=(X=1,Y=1,Z=1))))
    WeightMinimum=0
    Weight=&RigVMFunction_MathFloatRemap(SourceMaximum=100, SourceMinimum=0, bClamp=true, Value=@DeformRig_GetDeltaRefTransform_FNC_1.OutTransform.Rotation.Yaw).Result
   )
   > &RigUnit_ParentConstraint(
    AdvancedSettings=(InterpolationType=Average,RotationOrderForFilter=XZY)
    Child=(Type=Bone,Name="foot_robot_l")
    Filter=(TranslationFilter=(bX=false,bY=false,bZ=false),RotationFilter=(bX=true,bY=false,bZ=false),ScaleFilter=(bX=True,bY=True,bZ=True))
    Parents=((Item=(Type=Bone,Name="calf_l"),Weight=1))
    Weight=1
   )
   > &RigUnit_ParentConstraint(
    AdvancedSettings=(InterpolationType=Average,RotationOrderForFilter=XZY)
    Child=(Type=Bone,Name="foot_robot_r")
    Filter=(TranslationFilter=(bX=false,bY=false,bZ=false),RotationFilter=(bX=true,bY=false,bZ=false),ScaleFilter=(bX=True,bY=True,bZ=True))
    Parents=((Item=(Type=Bone,Name="calf_r"),Weight=1))
    Weight=1
   )
   > &RigUnit_ParentConstraint(
    AdvancedSettings=(InterpolationType=Average,RotationOrderForFilter=XZY)
    Child=(Type=Bone,Name="robot_calf_l")
    Filter=(TranslationFilter=(bX=false,bY=false,bZ=false),RotationFilter=(bX=true,bY=true,bZ=false),ScaleFilter=(bX=True,bY=True,bZ=True))
    Parents=((Item=(Type=Bone,Name="foot_robot_l"),Weight=1))
    Weight=1
   )
   > &RigUnit_ParentConstraint(
    AdvancedSettings=(InterpolationType=Average,RotationOrderForFilter=XZY)
    Child=(Type=Bone,Name="robot_calf_r")
    Filter=(TranslationFilter=(bX=false,bY=false,bZ=false),RotationFilter=(bX=true,bY=true,bZ=false),ScaleFilter=(bX=True,bY=True,bZ=True))
    Parents=((Item=(Type=Bone,Name="foot_robot_r"),Weight=1))
    Weight=1
   )
  ]}
 RigVMFunctionLibrary {nodes [
   DeformRig_GetDeltaRefTransform_FNC = RigVMCollapseNode(ExecuteContext=(), Item=(Type=Bone))
   Legs = RigVMCollapseNode(ExecuteContext=())
  ]}
 RigVMFunction_Sequence_1_ContainedGraph {nodes [
   n1 = RigVMFunctionEntryNode()
   RigVMFunction_Sequence = &RigVMFunction_Sequence(ExecuteContext=@n1.ExecuteContext)
   n2 = &RigVMFunction_Sequence(ExecuteContext=@RigVMFunction_Sequence.B)
   RigVMFunctionReturnNode(A=@RigVMFunction_Sequence.A, B=@n2.A, C=@n2.B)
  ]}
 RigVMModel {nodes [
   n1 = &RigUnit_PrepareForExecution()
   &RigUnit_HierarchyImportFromSkeleton(NameSpace=None, bIncludeCurves=False, bIncludeMeshSockets=False, bIncludeVirtualBones=True, ExecutePin=@n1.ExecutePin)
   RigUnit_BeginExecution = &RigUnit_BeginExecution()
   Legs = RigVMFunctionReferenceNode(ExecuteContext=@RigUnit_BeginExecution.ExecutePin)
   RigVMFunction_MathBoolToFloat = &RigVMFunction_MathBoolToFloat(Value=RigVMVariableNode() {variable=FEIdle}.Value)
   AlphaInterp = &RigVMFunction_AlphaInterp(
    InterpSpeedDecreasing=10
    InterpSpeedIncreasing=10
    OutRange=(Min=1,Max=0)
    Scale=1
    bClampResult=true
    bMapRange=true
    Value=@RigVMFunction_MathBoolToFloat.Result
   )
   > &RigUnit_RotationConstraintLocalSpaceOffset(
    AdvancedSettings=(InterpolationType=Average,RotationOrderForFilter=XZY)
    Child=(Type=Bone,Name="upperarm_twist_01_r")
    Filter=(bX=True,bY=True,bZ=True)
    Parents=((Item=(Type=Bone,Name="upperarm_r"),Weight=1))
    Weight=@AlphaInterp.Result
   )
   > &RigUnit_RotationConstraintLocalSpaceOffset(
    AdvancedSettings=(InterpolationType=Average,RotationOrderForFilter=XZY)
    Child=(Type=Bone,Name="upperarm_twist_02_r")
    Filter=(bX=True,bY=True,bZ=True)
    Parents=((Item=(Type=Bone,Name="upperarm_r"),Weight=1))
    Weight=@AlphaInterp.Result
   )
   > &RigUnit_RotationConstraintLocalSpaceOffset(
    AdvancedSettings=(InterpolationType=Average,RotationOrderForFilter=XZY)
    Child=(Type=Bone,Name="upperarm_twist_01_l")
    Filter=(bX=True,bY=True,bZ=True)
    Parents=((Item=(Type=Bone,Name="upperarm_l"),Weight=1))
    Weight=@AlphaInterp.Result
   )
   > &RigUnit_RotationConstraintLocalSpaceOffset(
    AdvancedSettings=(InterpolationType=Average,RotationOrderForFilter=XZY)
    Child=(Type=Bone,Name="upperarm_twist_02_l")
    Filter=(bX=True,bY=True,bZ=True)
    Parents=((Item=(Type=Bone,Name="upperarm_l"),Weight=1))
    Weight=@AlphaInterp.Result
   )
   > &RigUnit_ParentConstraint(
    AdvancedSettings=(InterpolationType=Average,RotationOrderForFilter=XZY)
    Child=(Type=Bone,Name="base_Laptop")
    Parents=((Item=(Type=Bone,Name="spine_03"),Weight=1))
    Weight=1
   )
   RigVMFunction_MathBoolOr = &RigVMFunction_MathBoolOr(A=RigVMVariableNode() {variable=targeting}.Value, B=RigVMVariableNode() {variable=firearm}.Value)
   Cast = &RigVMFunction_MathBoolToFloat(Value=@RigVMFunction_MathBoolOr.Result)
   KalmanFilter = &RigVMFunction_KalmanFloat(BufferSize=7, Value=@Cast.Result) {template="KalmanFilter::Execute(in Value,in BufferSize,out Result)"}
   RigVMFunction_AnimEasing = /Script/RigVM.RigVMFunction_AnimEasing(SourceMaximum=1, SourceMinimum=0, TargetMaximum=1, TargetMinimum=0, Type=CubicEaseInOut, Value=@KalmanFilter.Result) {is=struct RigVMUnitNode}
   > &RigUnit_ModifyTransforms(
    ItemToModify="((Item=(Type=Bone,Name=\"dyn_Laptop_03\"),Transform=(Rotation=(X=0,Y=-0.7373,Z=0,W=0.6756),Translation=(X=0,Y=0,Z=0),Scale3D=(X=1,Y=1,Z=1))),(Item=(Type=Bone,Name=\"dyn_Laptop_02\"),Transform=(Rotation=(X=0,Y=0.0872,Z=0,W=0.9962),Translation=(X=0,Y=0,Z=-4),Sca ...(truncated)"
    WeightMinimum=0
    Weight=@RigVMFunction_AnimEasing.Result
   )
   addFKCtrl = RigVMFunctionReferenceNode(
    Custom_FKCtrl=()
    ExecuteContext=()
    Options_FKCtrl="(pelvis_1_#1=false,spine_3_#2=true,clavicle_5_#3=false,arm_7_#4=false,leg_9_#5=true,head_11_46DC15EF465CD2 ...(truncated)"
    Shape_Settings="(pelvis_2_#6=(bVisible=True,Name=\"Hexagon_Thin\",Color=(R=0.0117,G=0.8281,B=1,A=1),Transform=(Rotation=(X=0,Y=-0.7071,Z=0,W=0.7071),Translation=(X=2,Y=0,Z=0),Scale3D=(X=5,Y=5,Z=7))),spine_4_#7=(bVi ...(truncated)"
   )
   setFKCtrl_1 = RigVMFunctionReferenceNode(ExecuteContext=(), setFK_Ctrl=false)
  ]}
]
variables [
 FEIdle {type=bool, default=False}
 firearm {type=bool, default=False}
 targeting {type=bool, default=False}
]
