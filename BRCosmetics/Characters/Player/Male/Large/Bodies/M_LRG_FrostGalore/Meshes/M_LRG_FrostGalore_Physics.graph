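# Physics asset for the large male FrostGalore cosmetic body, with capsule bodies for pelvis, spine, legs and dynamic "dyn" bones.
# Two three-segment dyn_Leg chains hang from tiny base_Leg bodies constrained to spine_02, with tight swing and almost no twist between segments.
# Thigh, calf and base_Leg joints rotate freely, while spine_02 to pelvis is limited.
# The dyn_sphere_calf bodies are short, wide capsules on free-rotating joints to the thighs, parallel to the calves.
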PhysicsAsset /BRCosmetics/Characters/Player/Male/Large/Bodies/M_LRG_FrostGalore/Meshes/M_LRG_FrostGalore_Physics
bodies {
 base_Leg_l [{shapeType=Capsule, length=1, radius=1, shapeName=base_Leg_l_capsule, rotation=<rot 90 0 0>}]
 base_Leg_r [{shapeType=Capsule, length=1, radius=1, shapeName=base_Leg_r_capsule, rotation=<rot 90 0 0>}]
 calf_l [{shapeType=Capsule, length=31, radius=10, shapeName=calf_r_capsule, center=<-23 0 1>, rotation=<rot 90 -3 0>}]
 calf_r [{shapeType=Capsule, length=31, radius=10, shapeName=calf_r_capsule, center=<23 0 -1>, rotation=<rot 90 -3 0>}]
 dyn_Leg_l_01 [{shapeType=Capsule, length=3.7914, radius=4.2004, shapeName=dyn_BartLeg_l_02_capsule, center=<2.8235 0 0.2024>, rotation=<rot 86.3541 0 0>}]
 dyn_Leg_l_02 [{shapeType=Capsule, length=3.7914, radius=4.2004, shapeName=dyn_BartLeg_l_02_capsule, center=<4.627 0 0.0875>, rotation=<rot 86.3541 0 0>}]
 dyn_Leg_l_03 [{
   shapeType=Capsule
   length=5.8245
   radius=5.0438
   shapeName=dyn_Leg_r_03_capsule
   center=<4.5275 -0.2997 2.7604>
   rotation=<rot 37.4053 -180 -173.4824>
  }]
 dyn_Leg_r_01 [{shapeType=Capsule, length=3.7914, radius=4.2004, shapeName=dyn_BartLeg_l_02_capsule, center=<-2.8235 0 -0.2024>, rotation=<rot 86.3541 0 0>}]
 dyn_Leg_r_02 [{shapeType=Capsule, length=3.7914, radius=4.2004, shapeName=dyn_BartLeg_l_02_capsule, center=<-4.627 0 -0.0875>, rotation=<rot 86.3541 0 0>}]
 dyn_Leg_r_03 [{
   shapeType=Capsule
   length=5.8245
   radius=5.0438
   shapeName=dyn_Leg_r_03_capsule
   center=<-4.5275 0.2997 -2.7604>
   rotation=<rot 37.4053 -180 -173.4824>
  }]
 dyn_sphere_calf_l [{shapeType=Capsule, length=3.5, radius=7.5, shapeName=dyn_sphere_calf_r_capsule, center=<1.2 2 0>}]
 dyn_sphere_calf_r [{shapeType=Capsule, length=3.5, radius=7.5, shapeName=dyn_sphere_calf_r_capsule, center=<-1.2 -2 0>}]
 pelvis [{shapeType=Capsule, length=20, radius=12.5509, center=<-3.1009 0 0>}]
 spine_02 [
  {shapeType=Capsule, length=20, radius=10, center=<-3.0619 4.7883 0>}
  {shapeType=Capsule, length=24.5287, radius=5.4713, shapeName=spine_02_capsule, center=<0.2396 19.1002 0>, rotation=<rot 90 36.7908 0>}
 ]
 thigh_l [{shapeType=Capsule, length=30, radius=7, shapeName=thigh_l_capsule, center=<-20 0 0>, rotation=<rot 90 0 0>}]
 thigh_r [{shapeType=Capsule, length=30, radius=7, shapeName=thigh_r_capsule, center=<20 0 0>, rotation=<rot 90 0 0>}]
}
constraints [
 {bone1Name=spine_02, bone2Name=pelvis, swing1LimitDegrees=45, swing2LimitDegrees=45, twistLimitDegrees=45}
 {
  bone1Name=base_Leg_l
  bone2Name=spine_02
  swing1LimitDegrees=45
  swing1Motion=Free
  swing2LimitDegrees=45
  swing2Motion=Free
  twistLimitDegrees=45
  twistMotion=Free
 }
 {bone1Name=dyn_Leg_l_01, bone2Name=base_Leg_l, swing1LimitDegrees=10, swing2LimitDegrees=20, twistLimitDegrees=1}
 {bone1Name=dyn_Leg_l_02, bone2Name=dyn_Leg_l_01, swing1LimitDegrees=10, swing2LimitDegrees=20, twistLimitDegrees=1}
 {bone1Name=dyn_Leg_l_03, bone2Name=dyn_Leg_l_02, swing1LimitDegrees=10, swing2LimitDegrees=20, twistLimitDegrees=1}
 {
  bone1Name=base_Leg_r
  bone2Name=spine_02
  swing1LimitDegrees=45
  swing1Motion=Free
  swing2LimitDegrees=45
  swing2Motion=Free
  twistLimitDegrees=45
  twistMotion=Free
 }
 {bone1Name=dyn_Leg_r_01, bone2Name=base_Leg_r, swing1LimitDegrees=10, swing2LimitDegrees=20, twistLimitDegrees=1}
 {bone1Name=dyn_Leg_r_02, bone2Name=dyn_Leg_r_01, swing1LimitDegrees=10, swing2LimitDegrees=20, twistLimitDegrees=1}
 {bone1Name=dyn_Leg_r_03, bone2Name=dyn_Leg_r_02, swing1LimitDegrees=10, swing2LimitDegrees=20, twistLimitDegrees=1}
 {
  bone1Name=thigh_l
  bone2Name=pelvis
  swing1LimitDegrees=45
  swing1Motion=Free
  swing2LimitDegrees=45
  swing2Motion=Free
  twistLimitDegrees=45
  twistMotion=Free
 }
 {
  bone1Name=thigh_r
  bone2Name=pelvis
  swing1LimitDegrees=45
  swing1Motion=Free
  swing2LimitDegrees=45
  swing2Motion=Free
  twistLimitDegrees=45
  twistMotion=Free
 }
 {
  bone1Name=calf_r
  bone2Name=thigh_r
  swing1LimitDegrees=45
  swing1Motion=Free
  swing2LimitDegrees=45
  swing2Motion=Free
  twistLimitDegrees=45
  twistMotion=Free
 }
 {
  bone1Name=calf_l
  bone2Name=thigh_l
  swing1LimitDegrees=45
  swing1Motion=Free
  swing2LimitDegrees=45
  swing2Motion=Free
  twistLimitDegrees=45
  twistMotion=Free
 }
 {
  bone1Name=dyn_sphere_calf_l
  bone2Name=thigh_l
  swing1LimitDegrees=45
  swing1Motion=Free
  swing2LimitDegrees=45
  swing2Motion=Free
  twistLimitDegrees=45
  twistMotion=Free
 }
 {
  bone1Name=dyn_sphere_calf_r
  bone2Name=thigh_r
  swing1LimitDegrees=45
  swing1Motion=Free
  swing2LimitDegrees=45
  swing2Motion=Free
  twistLimitDegrees=45
  twistMotion=Free
 }
]
