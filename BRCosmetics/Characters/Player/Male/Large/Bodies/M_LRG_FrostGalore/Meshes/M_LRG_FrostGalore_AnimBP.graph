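# Animation Blueprint for the FrostGalore large male cosmetic body, adding procedural and physics motion to the attached parent's pose.
# AnimGraph feeds the copied pose through a Control Rig, a spine_03-space RigidBody node, then sprung AnimDynamics on two arm-wire bones, each relative to the opposite-named upper arm.
# EventGraph updates gravity and rigid-body settings variables, plus firearm-equipped and front-end idle flags.
# Those flags and targeting state drive the Control Rig.
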AnimBlueprint /BRCosmetics/Characters/Player/Male/Large/Bodies/M_LRG_FrostGalore/Meshes/M_LRG_FrostGalore_AnimBP
target_skeleton=&/M_LRG_FrostGalore_Skeleton
graphs [
 AnimGraph {nodes [
   n1 = Control Rig!(
    Source=Copy Pose From Mesh() {bUseAttachedParent=true}
    Alpha
    targeting:bool=Get bIsTargeting()
    firearm:bool=Get Firearm Equipped() -> (Firearm Equipped:bool)
    FEIdle:bool=Get FEIdle() -> (FEIdle:bool)
   ) -> (Pose) {controlRigAssetReference {blueprintRigClass=&/M_LRG_FrostGalore_CtrlRig.M_LRG_FrostGalore_CtrlRig_C}}
   n2 = RigidBody!(
    ComponentPose=Local To Component(LocalPose=@n1)
    Alpha
    OverrideWorldGravity:Vector="0,0,0"
    ExternalForce
    SimSpaceSettings:Sim Space Settings Structure=(WorldAlpha=0,VelocityScaleZ=1,DampingAlpha=1,MaxLinearVelocity=10000,MaxAngularVelocity=10000,MaxLinearAcceleration=10000,MaxAngularAcceleration=10000,ExternalLinearDragV=(X=0,Y=0,Z=0),ExternalLinearVelocity=(X=0,Y=0,Z=0),ExternalAngularVelocity=(X=0,Y=0,Z=0))
   ) -> (Pose) {
    bClampLinearTranslationLimitToRefPose=true
    bOverrideWorldGravity=true
    overridePhysicsAsset=&/M_LRG_FrostGalore_Physics
    simulationSpace=BaseBoneSpace
    baseBoneRef {boneName=spine_03}
   }
   n3 = Anim Dynamics - Bone: dyn_ArmWire_l(ComponentPose=@n2, Alpha=1, ExternalForce="0,0,0") {
    angularDampingOverride=1
    angularSpringConstant=20000
    bAngularSpring=true
    bOverrideAngularDamping=true
    bUseGravityOverride=true
    gravityScale=0.35
    simulationSpace=BoneRelative
    boundBone {boneName=dyn_ArmWire_l}
    componentAppliedLinearAccClamp=<4000 4000 4000>
    componentLinearAccScale=<1 1 1>
    componentLinearVelScale=<1 1 1>
    gravityOverride {z=-980}
    physicsBodyDefinitions [{
      boundBone {boneName=dyn_ArmWire_l}
      boxExtents=<30 40 40>
      constraintSetup {angularLimitsMax=<25 35 25>, angularLimitsMin=<-25 -35 -25>, angularTarget=<1 0 0>}
      localJointOffset=<6 0 0>
     }]
    relativeSpaceBone {boneName=upperarm_r}
   }
   n4 = Anim Dynamics - Bone: dyn_ArmWire_r(ComponentPose=@n3, Alpha=1, ExternalForce="0,0,0") {
    angularDampingOverride=1
    angularSpringConstant=20000
    bAngularSpring=true
    bOverrideAngularDamping=true
    bUseGravityOverride=true
    gravityScale=0.35
    simulationSpace=BoneRelative
    boundBone {boneName=dyn_ArmWire_r}
    componentAppliedLinearAccClamp=<4000 4000 4000>
    componentLinearAccScale=<1 1 1>
    componentLinearVelScale=<1 1 1>
    gravityOverride {z=-980}
    physicsBodyDefinitions [{
      boundBone {boneName=dyn_ArmWire_r}
      boxExtents=<30 40 40>
      constraintSetup {angularLimitsMax=<25 35 25>, angularLimitsMin=<-25 -35 -25>, angularTarget=<1 0 0>}
      localJointOffset=<6 0 0>
     }]
    relativeSpaceBone {boneName=upperarm_l}
   }
   Output Pose(Result=Component To Local(ComponentPose=@n4)) {layerGroup=None}
  ]}
 EventGraph {nodes [
   Event BlueprintUpdateAnimation()
   > n1 = Sequence()
   > n2 = "Gravity Override State Function (Parameter Asset) (Nativized)"(
    CharacterPartAnimInstance=Self-Reference()
    GameTimeInSeconds=GetGameTimeInSeconds()
    DebugJointName=spine_05
    Parameters=&/M_LRG_FrostGalore_Parameters
   )
   > Set Gravity Override(Gravity Override=@n2)
   > Set RBAN Settings Out(RBAN Settings Out=@n2.OutRigidBodyAnimNodeParameters)
   > n3 = Cast To FortPlayerAnimInstance(Object=GetAnimInstance(self=GetOwnerSkeletalMeshComponent()))
   > Set Firearm Equipped(Firearm Equipped:bool=false) -> (Output_Get:bool)
   > n4 = Switch on EFortWeaponCoreAnimation!(execute, Selection:EFortWeaponCoreAnimation Enum=Get WeaponCoreAnim(self=@n3)) -> (
    Melee:Exec
    Pistol:Exec
    Shotgun:Exec
    PaperBlueprint:Exec
    Rifle:Exec
    MeleeOneHand:Exec
    MachinePistol:Exec
    RocketLauncher:Exec
    GrenadeLauncher:Exec
    GoingCommando:Exec
    AssaultRifle:Exec
    TacticalShotgun:Exec
    SniperRifle:Exec
    TrapPlacement:Exec
    ShoulderLauncher:Exec
    AbilityDecoTool:Exec
    Crossbow:Exec
    C4:Exec
    RemoteControl:Exec
    DualWield:Exec
    AR_BullPup:Exec
    AR_ForwardGrip:Exec
    MedPackPaddles:Exec
    SMG_P90:Exec
    AR_DrumGun:Exec
    Consumable_Small:Exec
    Consumable_Large:Exec
    Balloon:Exec
    MountedTurret:Exec
    CreativeTool:Exec
    ExplosiveBow:Exec
    AshtonIndigo:Exec
    AshtonChicago:Exec
    MeleeDualWield:Exec
    Unarmed:Exec
   )
   Set Firearm Equipped(
    execute=[
     @n4.AR_DrumGun
     @n4.AssaultRifle
     @n4.GrenadeLauncher
     @n4.MachinePistol
     @n4.Pistol
     @n4.Rifle
     @n4.RocketLauncher
     @n4.SMG_P90
     @n4.Shotgun
     @n4.SniperRifle
     @n4.TacticalShotgun
    ]
    Firearm Equipped=1
   )
   n5 = Get bIsPlayingEmote()
   n6 = Get bIsSkyDiving()
   Set FEIdle(execute=@n1.then_1, FEIdle:bool=AND Boolean!(A=Get bIsInFrontEnd(), B=NOT Boolean(A=@n6), C:bool=NOT Boolean(A=@n5)) -> (ReturnValue)) -> (Output_Get:bool)
  ]}
]
variables [
 FEIdle {type=boolean}
 Firearm Equipped {type=boolean}
 Gravity Override {type=Vector}
 RBAN Settings Out {type=FortRigidBodyAnimNodeParameters}
]
